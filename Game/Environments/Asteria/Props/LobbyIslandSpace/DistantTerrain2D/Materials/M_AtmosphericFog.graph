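# Translucent unlit material for atmospheric fog on the lobby island's 2D distant terrain, drawing a soft, animated, hollow elliptical haze.
# Two time-panned flame-noise samples, on radial UVs around Spiral_UV_Center, combine by Max and lightly perturb a stretched SphereMask.
# The result is sharpened, faded out toward the mask's center, clamped and scaled by an Opacity parameter.
# An EmissiveColor parameter supplies a tint color.
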Material /Game/Environments/Asteria/Props/LobbyIslandSpace/DistantTerrain2D/Materials/M_AtmosphericFog
graph {
 properties {blendMode=BLEND_Translucent, pixelDepthOffsetMode=PDOM_Legacy, shadingModel=MSM_Unlit, usedShadingModels=Unlit}
 nodes [
  n1 = TextureCoordinate()
  ConstantBiasScale(None=@n1) {bias=-0.5, scale=2}
  Constant() {r=5}
  n2 = Constant2Vector() {g=2.5, r=1.25}
  n3 = SphereMask(A=Multiply(A=@n1, B=@n2), B=Multiply(A=@n2, B=Constant2Vector() {g=0.52, r=0.52})) {attenuationRadius=0.85, hardnessPercent=0}
  n4 = Constant2Vector() {g=3, r=2}
  n5 = ComponentMask(None=VectorParameter() {group=PanningEmissive, parameterName=Spiral_UV_Center, defaultValue=<rgba 0.5 0.5 0 1>}.RGB) {g=true, r=true}
  n6 = MaterialFunctionCall!(Vector or UVs=Subtract(A=Multiply(A=@n1, B=@n4) {constB=2}, B=Multiply(A=@n5, B=@n4) {constB=2}) {constB=0.5}) {materialFunction=/Game/Athena/SafeZone/S11/MF_VectorToRadialValueFast}
  n7 = Time()
  n8 = Constant2Vector() {g=0.04, r=0.01}
  n9 = Power(Base=Frac(None=Multiply(A=@n7, B=@n8))) {constExponent=1.2}
  n10 = Multiply(A=Time()) {constB=-0.02}
  n11 = Max(
   A=TextureSample(UVs=Add(A=@n6.Result, B=@n9)) {texture=/Game/Effects/FogVolume/noise_flamy_Tile}.RGB
   B=TextureSample(UVs=Add(A=@n6.Result, B=@n10)) {texture=&noise_flamy_Tile}.RGB
  )
  n12 = Saturate(None=Add(A=@n3, B=Multiply(A=@n11) {constB=0.1}) {constB=0})
  n13 = Saturate(None=Power(Base=SmoothStep(Value=@n12)) {constExponent=10})
  n14 = Power(Base=SmoothStep(Value=@n3) {constMin=0.5}) {constExponent=0.5}
  Multiply(
   A=Clamp(None=LinearInterpolate(A=@n13, Alpha=@n14) {constB=0}) {maxDefault=0.5}
   B=ScalarParameter() {defaultValue=0.5, parameterName=Opacity}
  ) {constB=0.5}
  TextureSample() {samplerType=SAMPLERTYPE_LinearGrayscale, texture=/Game/Effects/Fort_Effects/Textures/T_DarkAndLight_Noise}
  VectorParameter() {parameterName=EmissiveColor, defaultValue=<rgba 1 1 1 1>}
 ]
}
