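# Translucent, unlit, two-sided material for a distant 2D island with animated waterfalls in the Asteria lobby island space.
# A channel-packed mask places water built from panning, height-blended noise with distorted UVs and brightened edges, its opacity varied by world-position noise.
# These layers combine with the HDR-remapped island texture and a hue-shifting radial airglow.
# The graph also holds camera-facing billboard offset math, though the named WPO value is a constant.
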Material /Game/Environments/Asteria/Props/LobbyIslandSpace/DistantTerrain2D/Materials/M_DistantIsland
graph {
 properties {blendMode=BLEND_Translucent, pixelDepthOffsetMode=PDOM_Legacy, shadingModel=MSM_Unlit, twoSided=true, usedShadingModels=Unlit}
 nodes [
  n1 = Add(A=NamedRerouteUsage(), B=NamedRerouteUsage())
  n2 = NamedRerouteUsage()
  n3 = NamedRerouteUsage()
  n4 = Power(Base=Add(A=NamedRerouteUsage(), B=Power(Base=@n3) {constExponent=1})) {constExponent=2.5}
  n5 = ScalarParameter() {defaultValue=0.2, parameterName=BrightenWaterEdge}
  Add(A=Add(A=Add(A=@n1, B=@n2), B=NamedRerouteUsage()), B=Multiply(A=@n4, B=@n5) {constB=0})
  n6 = Add(A=NamedRerouteUsage(), B=NamedRerouteUsage())
  n7 = NamedRerouteUsage()
  n8 = Constant3Vector() {constant=<rgba 0.2304 0.3524 1 0>}
  Add(A=Add(A=Add(A=@n6, B=@n7), B=NamedRerouteUsage()), B=Multiply(A=Multiply(A=@n8, B=@n4), B=@n5)) {constB=0}
  n9 = TextureSample() {
   samplerType=SAMPLERTYPE_Masks
   texture=/Game/Environments/Asteria/Props/LobbyIslandSpace/DistantTerrain2D/Textures/T_AsteriaIsland_Waterfall_M
  }
  NamedRerouteDeclaration(None=@n9.R) {name=WaterMask_R, nodeColor=<rgba 1 0 0.091 1>}
  NamedRerouteDeclaration(None=@n9.B) {name=WaterMask_B, nodeColor=<rgba 0 0.0691 1 1>}
  n10 = Power(Base=NamedRerouteUsage()) {constExponent=0.5}
  n11 = MaterialFunctionCall(Float2=TextureCoordinate()) {materialFunction=&BreakOutFloat2Components}
  n12 = OneMinus(None=Abs(None=ConstantBiasScale(None=@n11.R) {bias=-0.5, scale=2}))
  n13 = ScalarParameter() {defaultValue=1, parameterName=Water_DistortionExp}
  n14 = AppendVector(A=LinearInterpolate(A=Power(Base=@n12, Exp=@n13), B=@n12, Alpha=NamedRerouteUsage()), B=@n11.G)
  n15 = Multiply(
   A=AppendVector(
    A=ScalarParameter() {defaultValue=2.5, parameterName=WaterNoise_Overlay_TilingU}
    B=ScalarParameter() {defaultValue=2.5, parameterName=WaterNoise_Overlay_TilingV}
   )
   B=@n14
  )
  n16 = OneMinus(None=TextureSampleParameter2D(UVs=Panner(Coordinate=@n15) {speedY=-0.02}) {parameterName=WaterNoise_Overlay, texture=/Game/Global/Textures/Default/Tileable_Patterns/T_Smoke_OUT}.RGB)
  n17 = Multiply(A=@n10, B=LinearInterpolate(A=@n16, Alpha=@n10))
  n18 = AppendVector(
   A=ScalarParameter() {defaultValue=2.5, parameterName=WaterNoise1_TilingU}
   B=ScalarParameter() {defaultValue=2.5, parameterName=WaterNoise1_TilingV}
  )
  n19 = TextureSampleParameter2D(UVs=Panner(Coordinate=Multiply(A=@n14, B=@n18)) {speedY=-0.007}) {
   parameterName=WaterNoise1
   samplerType=SAMPLERTYPE_LinearGrayscale
   texture=/Game/Global/Textures/Default/Tileable_Patterns/T_StormCloudDetail01_M
  }
  n20 = Multiply(
   A=@n14
   B=AppendVector(
    A=ScalarParameter() {defaultValue=5, parameterName=WaterNoise2_TilingU}
    B=ScalarParameter() {defaultValue=5, parameterName=WaterNoise2_TilingV}
   )
  )
  n21 = LinearInterpolate(
   A=@n19.RGB
   B=TextureSampleParameter2D(UVs=Panner(Coordinate=@n20) {speedX=0.005, speedY=-0.005}) {parameterName=WaterNoise2, texture=&T_Smoke_OUT}.RGB
   Alpha=Power(Base=NamedRerouteUsage()) {constExponent=0.25}
  )
  n22 = Constant()
  n23 = Multiply(
   A=Multiply(A=MaterialFunctionCall(Transition Phase=@n17, Height Texture=@n21, Contrast=@n22) {materialFunction=&HeightLerp}.Alpha)
   B=@n16
  )
  n24 = ScalarParameter() {defaultValue=0.5, parameterName=WaterOpacity}
  n25 = Divide(
   A=ComponentMask(None=WorldPosition() {worldPositionShaderOffset=WPT_ExcludeAllShaderOffsets}.XYZ) {r=true}
   B=ScalarParameter() {defaultValue=35000, parameterName=WorldNoise_Size}
  )
  n26 = Saturate(None=LinearInterpolate(
    A=ScalarParameter() {defaultValue=0.15, parameterName=OpacityContrast_1}
    B=ScalarParameter() {defaultValue=0.5, parameterName=OpacityContrast_2}
    Alpha=TextureSample(UVs=@n25) {texture=&T_Inky_Smoke_Tile}.R
   ))
  NamedRerouteDeclaration(None=Multiply(A=Multiply(A=@n23, B=@n24) {constB=0.5}, B=NamedRerouteDeclaration(None=@n26) {name=NoiseForOpacity, nodeColor=<rgba 0 1 0.0625 1>})) {name=WaterfallOpacity, nodeColor=<rgba 0 1 0.4375 1>}
  NamedRerouteDeclaration(None=Multiply(A=@n23, B=VectorParameter() {parameterName=WaterColor, defaultValue=<rgba 0.2233 0.2492 0.5104 1>}.RGB)) {name=WaterfallColor, nodeColor=<rgba 0 1 0.4375 1>}
  n27 = Multiply(
   A=ComponentMask(None=TextureCoordinate() {uTiling=4, vTiling=4}) {r=true}
   B=ScalarParameter() {defaultValue=1.2, parameterName=WaterNoise_Edge_TilingV}
  ) {constB=0.5}
  n28 = NamedRerouteUsage()
  n29 = Multiply(A=ComponentMask(None=@n28) {g=true}, B=ScalarParameter() {defaultValue=0.5, parameterName=WaterNoise_Edge_TilingU}) {constB=0.5}
  n30 = TextureSampleParameter2D(UVs=Panner(Coordinate=AppendVector(A=@n27, B=@n29)) {speedX=0.01, speedY=-0.075}) {parameterName=WaterNoise3, texture=/Game/Effects/FogVolume/noise_flamy_Tile}
  n31 = Power(Base=@n28)
  n32 = MaterialFunctionCall(In=Add(A=Multiply(A=@n30.R, B=@n28), B=Multiply(A=@n31) {constB=0.5}), Contrast=Constant() {r=0.05}) {materialFunction=&CheapContrast}
  NamedRerouteDeclaration(None=@n32.Result) {name=WaterEdgeOpacity, nodeColor=<rgba 1 0 0.3516 1>}
  NamedRerouteDeclaration(None=Multiply(A=@n32.Result, B=VectorParameter() {parameterName=WaterEdgeColor, defaultValue=<rgba 0.3185 0.355 0.5365 1>}.RGB) {constB=0.5}) {name=WaterEdgeColor, nodeColor=<rgba 0 0.4609 1 1>}
  n33 = TextureSampleParameter2D(UVs=Add(A=TextureCoordinate() {desc=1x2, vTiling=2}, B=Constant2Vector() {g=0.5})) {parameterName=IslandTexture, texture=/Game/Environments/Asteria/Props/LobbyIslandSpace/DistantTerrain2D/Textures/T_AsteriaIsland_Waterfall_D}
  n34 = LinearInterpolate(
   A=ScalarParameter() {defaultValue=0.25, parameterName=EmissiveStrength_LDR}
   B=ScalarParameter() {defaultValue=0.35, parameterName=EmissiveStrength_LDR_Snow}
   Alpha=SmoothStep(Value=Multiply(A=@n33.G)) {constMin=0.15}
  )
  n35 = ScalarParameter() {defaultValue=1, parameterName=EmissiveStrength}
  NamedRerouteDeclaration(None=MaterialFunctionCall(LDR Input=Multiply(A=@n33.RGB, B=@n34) {constB=0.25}, HDR Tint=Multiply(A=@n33.RGB, B=@n35)) {materialFunction=&DeriveHDRfromLDR}.RemappedRange) {name=IslandTop, nodeColor=<rgba 1 0 0.3281 1>}
  NamedRerouteDeclaration(None=Constant()) {name=WPO, nodeColor=<rgba 0 0.0859 1 1>}
  n36 = OneMinus(None=Add(A=NamedRerouteUsage(), B=NamedRerouteUsage()))
  n37 = Constant2Vector() {g=2, r=1}
  n38 = TextureCoordinate()
  n39 = Constant2Vector() {g=0.4, r=0.5}
  n40 = MaterialFunctionCall(Vector or UVs=ConstantBiasScale(None=@n38) {bias=-0.5, scale=2}) {materialFunction=&VectorToRadialValue}
  n41 = Add(A=@n40.Vector Converted to Angle, B=Multiply(A=Time(), B=ScalarParameter() {defaultValue=0.25, parameterName=Airglow_HueShiftSpeed}))
  n42 = TextureSampleParameter2D(UVs=Multiply(A=@n40.Radial Coordinates) {constB=2}) {
   parameterName=AirGlowTexture
   texture=/Game/Characters/Player/Female/Medium/Bodies/F_MED_Rebirth_Soldier/Skins/Shiny_Jacket/Textures/T_Gradient_ShinyJacket
  }
  n43 = MaterialFunctionCall(
   Hue Shift Percentage=Multiply(A=Frac(None=@n41), B=ScalarParameter() {defaultValue=1, parameterName=Airglow_HueShiftAmount})
   Texture=@n42.RGB
  ) {materialFunction=&HueShift}
  n44 = Add(
   A=SphereMask(A=Multiply(A=@n38, B=@n37), B=Multiply(A=@n37, B=@n39)) {attenuationRadius=0.7, hardnessPercent=0}
   B=Multiply(A=ComponentMask(None=@n43.Result) {g=true}) {constB=0.075}
  )
  n45 = Multiply(A=SmoothStep(Value=Saturate(None=@n44)) {constMin=0.42}, B=ScalarParameter() {defaultValue=0.5, parameterName=Airglow_Amount})
  n46 = LinearInterpolate(A=Constant3Vector() {constant=<rgba 0.2341 0.1576 0.776 0>}, B=@n43.Result) {constAlpha=0.2}
  NamedRerouteDeclaration(None=Multiply(A=@n36, B=Multiply(A=@n45, B=@n46))) {name=AirGlow, nodeColor=<rgba 1 0.7266 0 1>}
  NamedRerouteDeclaration(None=Multiply(A=@n45, B=@n36)) {name=AirGlow_Opacity, nodeColor=<rgba 0.3828 1 0 1>}
  NamedRerouteDeclaration(None=@n9.G) {name=WaterMask_G, nodeColor=<rgba 0 0.7187 1 1>}
  NamedRerouteUsage()
  n47 = Subtract(A=WorldPosition().XYZ, B=ActorPositionWS())
  n48 = Normalize(VectorInput=Subtract(A=ActorPositionWS(), B=CameraPositionWS()))
  n49 = Constant3Vector() {constant=<rgba 0 0 1 0>}
  n50 = CrossProduct(A=@n48, B=@n49)
  Subtract(
   A=Add(
    A=MaterialFunctionCall(VectorToTransform=@n47, BasisX=@n48, BasisY=@n50, BasisZ=@n49) {materialFunction=&Transform3x3Matrix}.Transformed Vector
    B=ActorPositionWS()
   )
   B=WorldPosition().XYZ
  )
 ]
}
