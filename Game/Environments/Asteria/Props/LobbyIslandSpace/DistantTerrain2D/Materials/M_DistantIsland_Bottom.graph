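# Distant island underside material, named for the Asteria lobby island's 2D distant terrain props, that computes its own simplified lighting in the graph.
# A cliff texture's red channel blends two colors, which are shaded by the pixel normal against a sky-atmosphere light direction weighted toward a parameter vector.
# A Fresnel-driven tint is added for atmospheric edge color, and the sum is scaled down.
# Two rock normal maps are combined with a whiteout blend.
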Material /Game/Environments/Asteria/Props/LobbyIslandSpace/DistantTerrain2D/Materials/M_DistantIsland_Bottom
graph {
 properties {pixelDepthOffsetMode=PDOM_Legacy, usedShadingModels=Default Lit}
 nodes [
  Constant() {r=0.8}
  Constant()
  LinearInterpolate()
  n1 = Fresnel() {exponent=1}
  MaterialFunctionCall(In=@n1, Contrast=Constant() {r=0.05}) {materialFunction=&CheapContrast}
  MaterialFunctionCall!(
   Normal_1=TextureSample() {samplerType=SAMPLERTYPE_Normal, texture=/Game/Environments/Asteria/Terrain/Textures/T_Asteria_FrozenRockCliff_01_N}.RGB
   Normal_2=TextureSample() {samplerType=SAMPLERTYPE_Normal, texture=/Game/Athena/Asteria/Environments/Water/WaterEdge/T_S_CP_ModularRock_Mountain_A_02_N}.RGB
  ) {materialFunction=/Game/Global/MaterialFunctions/MF_NormalBlend_Whiteout}
  n2 = LinearInterpolate(
   A=VectorParameter() {parameterName=Color, defaultValue=<rgba 0.0298 0.0302 0.0365 1>}.RGB
   B=VectorParameter() {parameterName=Color2, defaultValue=<rgba 0.3594 0.351 0.3407 1>}.RGB
   Alpha=TextureSample(UVs=TextureCoordinate()) {texture=/Game/Environments/Asteria/Terrain/Textures/T_Asteria_FrozenRockCliff_02_D}.R
  )
  n3 = LinearInterpolate(A=SkyAtmosphereLightDirection(), B=VectorParameter() {parameterName=Directional Light, defaultValue=<rgba -1 0 1 1>}.RGB) {constAlpha=0.8}
  n4 = DotProduct(A=ComponentMask(None=Normalize(VectorInput=@n3)) {b=true, g=true, r=true}, B=PixelNormalWS())
  n5 = Constant() {r=2}
  n6 = Saturate(None=Add(A=MaterialFunctionCall(In=@n4, Contrast=@n5) {materialFunction=&CheapContrast}.Result) {constB=0.01})
  n7 = VectorParameter() {parameterName=FresnelColor, defaultValue=<rgba 0.0382 0.0583 0.1719 1>}
  n8 = Saturate(None=LinearInterpolate(Alpha=@n1) {constA=-3})
  Multiply(A=Add(A=Multiply(A=@n2, B=@n6), B=Multiply(A=@n7.RGB, B=@n8)) {constB=0}) {constB=0.4}
 ]
}
