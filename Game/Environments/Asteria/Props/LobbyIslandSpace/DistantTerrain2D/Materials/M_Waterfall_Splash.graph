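# Additive unlit material for waterfall splash foam in the lobby island's distant 2D terrain.
# Two downward-offset panning foam textures are contrast-adjusted and multiplied into a height map for HeightLerp, driven by vertical and side-edge UV falloff.
# The resulting alpha multiplies a tintable WaterfallColor.
# Other branches build a cone and sphere-masked foam blend and world-XY noise opacity with depth fade, though the graph shows no final output connections.
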Material /Game/Environments/Asteria/Props/LobbyIslandSpace/DistantTerrain2D/Materials/M_Waterfall_Splash
graph {
 properties {blendMode=BLEND_Additive, pixelDepthOffsetMode=PDOM_Legacy, shadingModel=MSM_Unlit, usedShadingModels=Unlit}
 nodes [
  n1 = MaterialFunctionCall(Float2=TextureCoordinate() {vTiling=2}) {materialFunction=&BreakOutFloat2Components}
  n2 = Power(Base=@n1.G) {constExponent=1}
  n3 = TextureSample(UVs=Panner(Coordinate=AppendVector(A=@n1.R, B=@n2)) {speedY=-0.03}) {texture=/Game/Packages/Fortnite_UniqueMaterials_1/Textures/FoamHeight}
  n4 = Saturate(None=MaterialFunctionCall(In=@n3.R, Contrast=Constant()) {materialFunction=&CheapContrast}.Result)
  n5 = MaterialFunctionCall(Float2=TextureCoordinate() {uTiling=0.4, vTiling=1.5}) {materialFunction=&BreakOutFloat2Components}
  n6 = Power(Base=@n5.G) {constExponent=1}
  n7 = TextureSample(UVs=Panner(Coordinate=AppendVector(A=@n5.R, B=@n6)) {speedY=-0.05}) {
   samplerType=SAMPLERTYPE_Masks
   texture=/Game/Environments/Asteria/Props/LobbyIslandSpace/DistantTerrain2D/Textures/T_GuavaTidalWave_TopFoam_H
  }
  n8 = MaterialFunctionCall(In=@n7.R, Contrast=Constant() {r=0.1}) {materialFunction=&CheapContrast}
  n9 = LinearInterpolate(A=@n4, B=@n8.Result, Alpha=Power(Base=Multiply(A=@n1.G) {constB=2}))
  n10 = MaterialFunctionCall(Float2=TextureCoordinate()) {materialFunction=&BreakOutFloat2Components}
  n11 = Constant() {r=1}
  n12 = Multiply(
   A=SphereMask(A=@n10.R, B=Constant() {r=0.5}) {attenuationRadius=0.5, hardnessPercent=0}
   B=OneMinus(None=SphereMask(A=@n10.G, B=@n11) {attenuationRadius=0.05, hardnessPercent=0})
  )
  Multiply(
   A=TextureSample() {texture=/Game/Athena/HUD/Vehicles/Materials/T_UI_ConeMask}.R
   B=Multiply(A=Multiply(A=@n9, B=@n12)) {constA=1, constB=0.2}
  )
  n13 = ScalarParameter() {parameterName=OpacityContrast_1}
  n14 = ScalarParameter() {defaultValue=2.7377, parameterName=OpacityContrast_2}
  n15 = WorldPosition() {worldPositionShaderOffset=WPT_ExcludeAllShaderOffsets}
  n16 = TextureSample(UVs=Divide(A=ComponentMask(None=@n15.XYZ) {g=true, r=true}, B=ScalarParameter() {defaultValue=79682.875, parameterName=WorldNoise_Size})) {texture=/Game/Global/Textures/Default/Noise/T_Tiling_Noise_01}
  n17 = NamedRerouteDeclaration(None=Saturate(None=LinearInterpolate(A=@n13, B=@n14, Alpha=@n16.R))) {name=NoiseForOpacity, nodeColor=<rgba 0 1 0.0625 1>}
  Multiply(A=DepthFade(Opacity=Multiply(A=@n17) {constB=0.5}, FadeDistance=ScalarParameter() {defaultValue=8834.1191, parameterName=DepthFade})) {constB=0.5}
  n18 = TextureCoordinate()
  n19 = OneMinus(None=ComponentMask(None=@n18) {g=true})
  n20 = ScalarParameter() {defaultValue=0.5, parameterName=FalloffOpacity}
  n21 = ComponentMask(None=@n18) {r=true}
  n22 = OneMinus(None=@n21)
  n23 = Power(Base=Multiply(A=Multiply(A=@n21, B=@n22)) {constB=4}, Exp=ScalarParameter() {defaultValue=1, parameterName=EdgeFalloff})
  n24 = MaterialFunctionCall(
   Transition Phase=Multiply(A=Power(Base=@n19, Exp=@n20), B=Saturate(None=@n23))
   Height Texture=Multiply(A=@n4, B=@n8.Result)
   Contrast=Constant()
  ) {materialFunction=&HeightLerp}
  Multiply(A=@n24.Alpha)
  Multiply(A=@n24.Alpha, B=Multiply(A=VectorParameter() {parameterName=WaterfallColor, defaultValue=<rgba 0.5013 0.5285 0.8021 1>}.RGB) {constB=0.5})
  Multiply(A=@n8.Result) {constB=0.2}
  n25 = ScalarParameter() {defaultValue=-1.0954, parameterName=NoiseContrast_1}
  n26 = ScalarParameter() {defaultValue=10.657, parameterName=NoiseContrast_2}
  NamedRerouteDeclaration(None=Saturate(None=LinearInterpolate(A=@n25, B=@n26, Alpha=@n16.R))) {name=NoiseHighContrast, nodeColor=<rgba 0 1 0.0156 1>}
  NamedRerouteDeclaration(None=@n16.R) {name=Noise, nodeColor=<rgba 1 0.6797 0 1>}
  OneMinus()
  Power(Base=@n3.R) {constExponent=1}
 ]
}
